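# Physics asset for the Cat Burglar Male two-handed pickaxe, providing bodies for its handle and dangling chains.
# The handle bone carries a capsule shaft and a box head, while four chain bones each get a small capsule.
# Chain a_02 links to a_01, while a_01, b_01 and mid each constrain directly to the handle.
# The a_01 joint locks twist and one swing axis, restricting how that chain root moves.
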PhysicsAsset /BRCosmetics/Weapons/FORT_Melee/Pickaxe_Cat_Burglar_Male/Meshes/Cat_Burglar_Male_Axe_PhysicsAsset
bodies {
 Melee_TwoHanded_Handle [
  {shapeType=Capsule, length=101.8021, radius=2.6553, center=<0.8695 -1.0607 26.7368>}
  {shapeType=Box, extentX=7.5, extentY=79.9883, extentZ=23.5571, center=<1.3195 0 81.6821>}
 ]
 dyn_chain_a_01 [{shapeType=Capsule, length=2, radius=2, center=<1 0 0>, rotation=<rot 90 0 0>}]
 dyn_chain_a_02 [{shapeType=Capsule, length=4, radius=2, center=<8 0 0>, rotation=<rot 90 -0.0099 -0.0099>}]
 dyn_chain_b_01 [{shapeType=Capsule, length=5, radius=2, center=<9 -0.0001 -0.4025>, rotation=<rot 84.9999 0 0>}]
 dyn_chain_mid [{shapeType=Capsule, length=4, radius=2, center=<10 0 0>, rotation=<rot 90 -0.0099 -0.0099>}]
}
constraints [
 {
  bone1Name=dyn_chain_a_01
  bone2Name=Melee_TwoHanded_Handle
  swing1LimitDegrees=45
  swing1Motion=Locked
  swing2LimitDegrees=90
  twistLimitDegrees=45
  twistMotion=Locked
 }
 {bone1Name=dyn_chain_a_02, bone2Name=dyn_chain_a_01, swing1LimitDegrees=75, swing2LimitDegrees=10, twistLimitDegrees=5}
 {bone1Name=dyn_chain_b_01, bone2Name=Melee_TwoHanded_Handle, swing1LimitDegrees=75, swing2LimitDegrees=10, twistLimitDegrees=5}
 {bone1Name=dyn_chain_mid, bone2Name=Melee_TwoHanded_Handle, swing1LimitDegrees=5, swing2LimitDegrees=90, twistLimitDegrees=5}
]
